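# Niagara system for the JoyGrid player emote that renders flickering sparks, flares with lights, and glow decals at two skeletal sockets.
# Three emitters split particles by partition between the sockets, with User.FXActive and User.FXColor separately controlling each side's visibility.
# Those two floats also reshape spark spread and direction and decal size.
# A system FlickerSpeed drives spark rate, speed and glow brightness, while platform checks adjust flare intensity.
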NiagaraSystem /BRCosmetics/Animation/Game/MainPlayer/Emotes/JoyGrid/VFX/NS_Emote_JoyGrid
dependencies {
 usedDataInterfaces [
  &NiagaraDataInterfaceArrayInt32
  &NiagaraDataInterfacePlatformSet
  &NiagaraDataInterfaceVector2DCurve
  &NiagaraDataInterfaceSpriteRendererInfo
  &NiagaraDataInterfaceArrayFloat
  &NiagaraDataInterfaceCurve
  &NiagaraDataInterfaceColorCurve
  &NiagaraDataInterfaceSocketReader
  &NiagaraDataInterfaceSkeletalMesh
 ]
 usedDynamicInputs [
  &LerpPosition
  &SelectFloatFromArray
  &ModuloInt
 ]
 usedModules [
  &self:SystemSpawnScript.SystemScriptSource.SystemScriptGraph.NiagaraNodeAssignment_0.SetVariables_#1
  &self:SystemSpawnScript.SystemScriptSource.SystemScriptGraph.NiagaraNodeAssignment_2.SetVariables_#2
  &self:SystemSpawnScript.SystemScriptSource.SystemScriptGraph.NiagaraNodeAssignment_3.SetVariables_#3
  &SystemState
  &BlendSpaceHelper
  &self:SystemSpawnScript.SystemScriptSource.SystemScriptGraph.NiagaraNodeAssignment_1.SetVariables_#4
 ]
}
emitters {
 Classic_Sparks {
  emitterUpdateScript [
   &self:InterpolateSpeedThresholdEmitter_01(
    Define Initial Value=false
    Rate Of Change=5
    Target Float Value:NiagaraFloat=&MakeCustomFloatFromBool(
     Bool=&SetBoolByFloatComparison(
      A=&Add_Float(
       A=&Add_Float(
        A=&Abs_Float(Float:NiagaraFloat=&MakeFloatFromVector(Vector=$Engine.Owner.Velocity, Channel=::NewEnumerator0))
        B=&Abs_Float(Float:NiagaraFloat=&MakeFloatFromVector(Vector=$Engine.Owner.Velocity, Channel=::NewEnumerator2))
       )
       B=&Abs_Float(Float:NiagaraFloat=&MakeFloatFromVector(Vector=$Engine.Owner.Velocity, Channel=::NewEnumerator1))
      )
      B=46.278
     )
     True Float=1
     False Float=0
    )
   ) {module=InterpolateSpeedThresholdEmitter_01}
   &EmitterState(Life Cycle Mode=::NewEnumerator1, Loop Delay=2.3833, Loop Duration=2, UseLoopDelay=false)
   &SpawnRate(SpawnRate=&Multiply_Float(A=400, B=$System.FlickerSpeed), Use Spawn Probability=false)
  ]
  particleSpawnScript [
   &PartitionParticles()
   &InitializeParticle(
    Color Mode=::NewEnumerator0
    Lifetime=&RandomRangeFloat(Minimum=0.25, Maximum=0.18)
    Position Mode=::NewEnumerator0
    Sprite Rotation Mode=::NewEnumerator2
    Sprite Size Max=<3 6>
    Sprite Size Min=<1 1>
    Sprite Size Mode=::NewEnumerator2
    Write Material Random=true
   )
   &SocketLocation(
    "Inherit Velocity (Scale)"=0
    SamplePosOffset=false
    Sampled Velocity Coordinate Space:ENiagaraCoordinateSpace=::Simulation
    Socket Index=&SelectIntFromArray(Direct Array Index=$Particles.PartitionParticles.Partition, Int Selection Array=DataInterface() {propertyValues {IntData [0,1]}})
    Socket Reader=$System.Socket Reader
    Socket Sampling Mode=::NewEnumerator3
    VelocitySpeedLimit=false
   ) {enabled=false}
   &SkeletalMeshLocation(
    Bone / Socket Index=&SelectIntFromArray(Direct Array Index=$Particles.PartitionParticles.Partition, Int Selection Array=DataInterface() {propertyValues {IntData [0,1]}})
    MaskSpawnGroup=false
    Mesh Sampling Type=::NewEnumerator5
    Orientation Sampling=::NewEnumerator1
    Sampled Position Offset=&MaskVector(
     Mask Source=1
     Mask Parameter=$Particles.PartitionParticles.Partition
     Mask=0
     Masked Vector=&MakeVector(X=0, Y=0, Z=0)
     Passthrough Vector=&MakeVector(X=0, Y=0, Z=0)
    )
    Sampled Position Offset Coordinate Space:ENiagaraSkelMeshTransforms=::NewEnumerator2
    Skeletal Mesh=$System.Skeletal Mesh Reader
    Socket Sampling Mode=::NewEnumerator5
   )
   &AddVelocity(
    Additional Quaternion Rotation=false
    Cone Angle=&MakeCustomFloatFromBool(
     Bool=&Boolean_OrOperation(
      Bool1=&SetBoolByFloatComparison(A=$User.FXActive, B=1)
      Bool2=&SetBoolByFloatComparison(A=$User.FXColor, B=1)
     )
     True Float=5
     False Float=50
    )
    Cone Axis=&MaskVector(
     Mask Source=1
     Mask Parameter=$Particles.PartitionParticles.Partition
     Mask=0
     Masked Vector=&MakeCustomVectorFromBool(Bool=&SetBoolByFloatComparison(A=$User.FXActive, B=1), True Vector=<1 0.1 0>, False Vector=<-1 1 0>)
     Passthrough Vector=&MakeCustomVectorFromBool(Bool=&SetBoolByFloatComparison(A=$User.FXColor, B=1), True Vector=<1 0.1 0>, False Vector=<-1 1 0>)
    )
    Module.UseConeFalloff=false
    Module.UseVelDistribution=false
    Override Seed=false
    Rotation Mode=::NewEnumerator3
    Rotation Quaternion:Quat4f=&MultiplyQuaternion(Quaternion A=$Particles.SkeletalMeshLocation.SampledRotation, Quaternion B=<0.8 0.8 0 0>)
    Velocity Mode=::NewEnumerator2
    Velocity Speed=&Multiply_Float(
     A=&Multiply_Float(
      A=&Lerp_Float(
       A=135
       B=215
       Alpha=&Waveform("Waveform [1]"=::NewEnumerator0, "[1] Frequency"=1.5, "[1] Amplitude Min/Max"=<0 1>)
      )
      B=&Multiply_Float(A=$System.FlickerSpeed, B=1.2)
     )
     B=&MakeCustomFloatFromBool(
      Bool=&Boolean_OrOperation(
       Bool1=&SetBoolByFloatComparison(A=$User.FXActive, B=1)
       Bool2=&SetBoolByFloatComparison(A=$User.FXColor, B=1)
      )
      True Float=&RandomRangeFloat(Minimum=0.6, Maximum=0.9)
      False Float=1
     )
    )
   ) {module=AddVelocity001}
   &ShapeLocation(
    Box Midpoint=<0.5 0.5 0>
    Box Size=&MakeCustomVectorFromBool(
     Bool=&Boolean_OrOperation(
      Bool1=&SetBoolByFloatComparison(A=$User.FXActive, B=1)
      Bool2=&SetBoolByFloatComparison(A=$User.FXColor, B=1)
     )
     True Vector=<0 0 0>
     False Vector=<9.1912 2 0>
    )
    MaskSpawnGroup=false
    Override Seed=false
    Rotation Coordinate Space=::World
    Shape Primitive=::NewEnumerator2
    Surface Only=false
    UseOwnerScale=true
   )
   &ShapeLocation(
    Handle Radius:NiagaraFloat=&MakeCustomFloatFromBool(
     Bool=&Boolean_OrOperation(
      Bool1=&SetBoolByFloatComparison(A=$User.FXActive, B=1)
      Bool2=&SetBoolByFloatComparison(A=$User.FXColor, B=1)
     )
     True Float=0
     False Float=0
    )
    Large Radius=&MakeCustomFloatFromBool(
     Bool=&Boolean_OrOperation(
      Bool1=&SetBoolByFloatComparison(A=$User.FXActive, B=1)
      Bool2=&SetBoolByFloatComparison(A=$User.FXColor, B=1)
     )
     True Float=2.5
     False Float=0
    )
    MaskSpawnGroup=false
    Non Uniform Scale=<1 3 0>
    Offset=&MakeCustomVectorFromBool(
     Bool=&Boolean_OrOperation(
      Bool1=&SetBoolByFloatComparison(A=$User.FXActive, B=1)
      Bool2=&SetBoolByFloatComparison(A=$User.FXColor, B=1)
     )
     True Vector=<0 -7 0>
     False Vector=<0 0 0>
    )
    Offset Coordinate Space=::Local
    Offset Mode=::NewEnumerator0
    Override Seed=false
    Shape Primitive=::NewEnumerator3
    UseOwnerScale=true
   ) {module=ShapeLocation002}
   &ShapeLocation(
    Box Size=<60 60 150>
    MaskSpawnGroup=true
    Offset=<0 0 58.7788>
    Offset Mode=::NewEnumerator0
    Override Seed=false
    Shape Primitive=::NewEnumerator2
    Spawn Group Mask=1
    Surface Only=false
    UseOwnerScale=true
   ) {enabled=false, module=ShapeLocation001}
   &AddVelocity(
    Velocity=&Lerp_Vector(
     A=&RandomRangeVector(Minimum=<-40 5 0>, Maximum=<40 20 0>)
     B=&RandomRangeVector(Minimum=<-60 5 0>, Maximum=<60 20 0>)
     Alpha=&UniformAOrBFloat(A=0.2, Distribution Weight=0, Evaluation Type=::NewEnumerator0, Recalculate Random Each Loop=true, Override Seed=false)
    )
    Velocity Speed Scale=&MakeCustomFloatFromBool(
     Bool=&Boolean_OrOperation(
      Bool1=&SetBoolByFloatComparison(A=$User.FXActive, B=1)
      Bool2=&SetBoolByFloatComparison(A=$User.FXColor, B=1)
     )
     True Float=2.5
     False Float=0
    )
   )
   &CameraOffset(Camera Offset Amount=-2)
  ]
  particleUpdateScript [
   &ParticleState()
   /SetVariables(
    Particles.Flicker:NiagaraFloat=&Waveform("Waveform [1]"=::NewEnumerator0, "[1] Frequency"=75, "[1] Amplitude Min/Max"=<0 2>)
    Particles.Flicker Size:NiagaraFloat=&Waveform("Waveform [1]"=::NewEnumerator0, "[1] Frequency"=75, "[1] Amplitude Min/Max"=<0.5 1>)
    Particles.Flicker Value:NiagaraFloat=&Waveform("Waveform [1]"=::NewEnumerator0, "[1] Frequency"=75, "[1] Amplitude Min/Max"=<2 8>)
   ) {isSetParameters=true}
   &ScaleSpriteSize(
    Scale Sprite Size Mode=::NewEnumerator0
    Uniform Scale Factor=&Lerp_Float(
     A=1
     B=&Multiply_Float(A=$Particles.Flicker Size, B=1)
     Alpha=&FloatFromCurve(FloatCurve=DataInterface() {propertyValues {Curve {keys [<at 0 0.1076>,<at 0.2485 0.6128>,<at 0.45 1>,<at 1 1>]}}})
    )
   )
   &ScaleColor(
    Scale Mode=::NewEnumerator1
    Scale RGBA=&Multiply_Vector4(
     A=&MakeVector4_FromLinearColor(LinearColor=&RandomRangeLinearColor(
       Minimum=&ColorFromCurve(ColorCurve=DataInterface() {propertyValues {
          AlphaCurve {keys [<at 0.122 1>]}
          BlueCurve {keys [<at 0 1.0417>,<at 0.21 0.9911>,<at 0.4079 0>,<at 0.6487 0>]}
          GreenCurve {keys [<at 0 7.4334>,<at 0.21 10.4768>,<at 0.4079 9.1355>,<at 0.6487 1.1926>]}
          RedCurve {keys [<at 0 20>,<at 0.21 16>,<at 0.4079 13>,<at 0.6487 9>]}
         }})
       Maximum=&ColorFromCurve(ColorCurve=DataInterface() {propertyValues {
          AlphaCurve {keys [<at 0.122 1>]}
          BlueCurve {keys [<at 0 3.3499>,<at 0.089 3.7389>,<at 0.2245 0>,<at 0.8118 0>]}
          GreenCurve {keys [<at 0 17.3523>,<at 0.089 9.9197>,<at 0.2245 2.1674>,<at 0.8118 1.159>]}
          RedCurve {keys [<at 0 20>,<at 0.089 16>,<at 0.2245 13>,<at 0.8118 9>]}
         }})
      ))
     B=&MakeVector4_FromLinearColor(LinearColor=&MakeLinearColorFromVectorAndFloat(
       "Vector (RGB)"=&Add_Vector(
        A=<1 1 1>
        B=/Niagara/DynamicInputs/TypeConversions/MakeVectorFromLinearColorHSV(Color:LinearColor=/Niagara/DynamicInputs/LinearColor/MakeLinearColorFromHSV(
          HSV From:NiagaraInt32=1
          HSV:Vector3f=?
          Hue:NiagaraFloat=0
          Saturation:NiagaraFloat=0
          Value:NiagaraFloat=&Lerp_Float(
           A=0
           B=$Particles.Flicker Value
           Alpha=&FloatFromCurve(FloatCurve=DataInterface() {propertyValues {Curve {keys [<at 0 0>,<at 0.5 0>,<at 0.55 0.9624>,<at 1 1>]}}})
          )
          Alpha:NiagaraFloat=1
         ))
       )
       "Float (Alpha)"=&Multiply_Float(
        A=&Lerp_Float(
         A=1
         B=$Particles.Flicker
         Alpha=&FloatFromCurve(FloatCurve=DataInterface() {propertyValues {Curve {keys [<at 0 0>,<at 0.25 0>,<at 0.35 1>,<at 1 1>]}}})
        )
        B=&MaskFloat(
         Mask Source=1
         Mask Parameter=$Particles.PartitionParticles.Partition
         Mask=0
         Masked Float=&Multiply_Float(A=1, B=$User.FXActive)
         Passthrough Float=&Multiply_Float(A=1, B=$User.FXColor)
        )
       )
      ))
    )
    ScaleRGBA=true
   )
   &GravityForce(Gravity=&RandomRangeVector(Minimum=<0 0 -161.3829>, Maximum=<0 0 -457.3416>))
   &InheritSourceMovement(Applied Position Scale=<-0.1 -0.3 0>, ApplyForce=false, ApplyPosition=true, ApplyPositionFalloff=false) {enabled=false}
   &InheritVelocity(
    Inherited Velocity Amount Scale=&MakeCustomVectorFromBool(
     Bool=&SetBoolByFloatComparison(A=&VectorLength(Vector=$User.OwnerLocalVelocity), B=$System.WalkForwardSpeed)
     True Vector=<0 0 0>
     False Vector=&MakeCustomVectorFromBool(
      Bool=&Boolean_OrOperation(
       Bool1=&SetBoolByFloatComparison(A=$User.FXActive, B=1)
       Bool2=&SetBoolByFloatComparison(A=$User.FXColor, B=1)
      )
      True Vector=<0 -0.02 0>
      False Vector=<-0.02 -0.01 0>
     )
    )
    Speed Limit=true
    UseSpeedThreshold=false
   )
   &SolveForcesAndVelocity(Clamp Velocity=false, Limit Acceleration=false)
   &ScaleSpriteSizeBySpeed(
    Max Scale Factor=&RandomRangeVector2D(Minimum=<0.22 18.0452>, Maximum=<6 11.7483>)
    Min Scale Factor=<0.15 1>
    Velocity Threshold=3800
   )
   /SetVariables(
    Particles.GlowColor=&MakeLinearColorFromVectorAndFloat(
     "Vector (RGB)"=&MakeVectorFromLinearColorRGB(Color=&ColorFromCurve(ColorCurve=DataInterface() {propertyValues {
         AlphaCurve {keys []}
         BlueCurve {keys [<at 0 0.4192 RCIM_Cubic>,<at 0.3932 0>,<at 0.8828 0>,<at 1 0 RCIM_Cubic>]}
         GreenCurve {keys [<at 0 0.8224 RCIM_Cubic>,<at 0.3932 0.5726>,<at 0.8828 0.2129>,<at 1 0.1589 RCIM_Cubic>]}
         RedCurve {keys [<at 0 1 RCIM_Cubic>,<at 0.3932 1>,<at 0.8828 1>,<at 1 1 RCIM_Cubic>]}
        }}))
     "Float (Alpha)"=&Multiply_Float(
      A=&Lerp_Float(
       A=1
       B=$Particles.Flicker
       Alpha=&FloatFromCurve(FloatCurve=DataInterface() {propertyValues {Curve {keys [<at 0 0>,<at 0.25 0>,<at 0.35 1>,<at 1 1>]}}})
      )
      B=&MaskFloat(
       Mask Source=1
       Mask Parameter=$Particles.PartitionParticles.Partition
       Mask=0
       Masked Float=&Multiply_Float(A=0.8, B=$User.FXActive)
       Passthrough Float=&Multiply_Float(A=0.8, B=$User.FXColor)
      )
     )
    )
    Particles.GlowSize=&Multiply_Vector2D(A=$Particles.SpriteSize, B=<8 4>)
   ) {isSetParameters=true}
  ]
  rendererClasses [&NiagaraSpriteRendererProperties]
  renderers [
   {rendererClass=&NiagaraSpriteRendererProperties}
   {rendererClass=&NiagaraSpriteRendererProperties, rendererIndex=1}
  ]
 }
 Decal {
  emitterUpdateScript [
   &EmitterState()
   &SpawnBurst_Instantaneous(Spawn Count=2, Use Spawn Probability=false, UseLoopCountLimit=false)
  ]
  particleSpawnScript [
   &PartitionParticles()
   &InitializeParticle(
    Color=<rgba 1.5 0.4205 0.0061 0.2696>
    Lifetime=$Emitter.CurrentLoopDuration
    Sprite Size=<77 110>
    Sprite Size Mode=::NewEnumerator4
    UsePositionOffset=false
    Write Material Random=true
   )
   &SpriteFacingAndAlignment(Sprite Facing=<0 0 1>, WriteAlignment=false, WriteFacing=true)
   &CameraOffset(Camera Offset Amount=-1)
  ]
  particleUpdateScript [
   &ParticleState()
   &SkeletalMeshLocation(
    Bone / Socket Index=$Particles.PartitionParticles.Partition
    MaskSpawnGroup=false
    Mesh Sampling Type=::NewEnumerator5
    Sampled Position Offset=<0 10 0>
    Skeletal Mesh=$System.Skeletal Mesh Reader
    Socket Sampling Mode=::NewEnumerator5
   )
   &SolveForcesAndVelocity(Clamp Velocity=false, Limit Acceleration=false)
   &ScaleColor(
    Scale Alpha=&MaskFloat(
     Mask Source=1
     Mask Parameter=$Particles.PartitionParticles.Partition
     Mask=0
     Masked Float=&Multiply_Float(
      A=1
      B=&MakeCustomFloatFromBool(
       Bool=&SetBoolByFloatComparison(A=$User.FXActive, B=0, Comparison Type=::NewEnumerator6)
       True Float=&Multiply_Float(A=$User.FXActive, B=-1)
       False Float=$User.FXActive
      )
     )
     Passthrough Float=&Multiply_Float(
      A=1
      B=&Multiply_Float(
       A=1
       B=&MakeCustomFloatFromBool(
        Bool=&SetBoolByFloatComparison(A=$User.FXColor, B=0, Comparison Type=::NewEnumerator6)
        True Float=&Multiply_Float(A=$User.FXColor, B=-1)
        False Float=$User.FXColor
       )
      )
     )
    )
    Scale RGB=&Multiply_VectorByFloat(Vector=<1 1 1>, Float=$System.FlickerSpeed)
    ScaleA=true
    ScaleRGB=true
   )
   /Niagara/Modules/Decal/Decal_Attributes(
    Decal Fade=0.4
    Decal Size:Vector3f=&MakeCustomVectorFromBool(
     Bool=&Boolean_OrOperation(
      Bool1=&SetBoolByFloatComparison(A=$User.FXActive, B=1)
      Bool2=&SetBoolByFloatComparison(A=$User.FXColor, B=1)
     )
     True Vector=<20 70 100>
     False Vector=<20 70 70>
    )
    Fade=1
    Size=1
   )
   &ScaleColor(
    Scale Alpha=&MaskFloat(
     Mask Source=1
     Mask Parameter=$Particles.PartitionParticles.Partition
     Mask=0
     Masked Float=&Multiply_Float(
      A=1
      B=&MakeCustomFloatFromBool(
       Bool=&SetBoolByFloatComparison(A=$User.FXActive, B=0, Comparison Type=::NewEnumerator6)
       True Float=&Multiply_Float(A=$User.FXActive, B=-1)
       False Float=$User.FXActive
      )
     )
     Passthrough Float=&Multiply_Float(
      A=1
      B=&Multiply_Float(
       A=1
       B=&MakeCustomFloatFromBool(
        Bool=&SetBoolByFloatComparison(A=$User.FXColor, B=0, Comparison Type=::NewEnumerator6)
        True Float=&Multiply_Float(A=$User.FXColor, B=-1)
        False Float=$User.FXColor
       )
      )
     )
    )
    ScaleA=true
    ScaleRGB=true
   ) {module=ScaleColor001}
  ]
  rendererClasses [&NiagaraDecalRendererProperties]
  renderers [{rendererClass=&NiagaraDecalRendererProperties, rendererIndex=0}]
 }
 Flash_Flare {
  emitterUpdateScript [
   &EmitterState(Inactive Response=::NewEnumerator2, Life Cycle Mode=::NewEnumerator1, Loop Duration=2, UseLoopDelay=false)
   &SpawnBurst_Instantaneous(Spawn Count=2, Spawn Group=2, Use Spawn Probability=false, UseLoopCountLimit=false)
   &SpawnRate(SpawnRate=8, Use Spawn Probability=false)
  ]
  particleSpawnScript [
   /SetVariables(Particles.VisibilityTag=&SpawnGroup()) {isSetParameters=true}
   &PartitionParticles()
   &InitializeParticle(
    Color=<rgba 1.5 0.9499 0.4324 1>
    Lifetime=&MaskFloatBySpawnGroup(Masked Float=0.36, Passthrough Float=0.36)
    Position=&SimulationPosition()
    Position Mode=::NewEnumerator1
    Position Offset=&MaskVectorBySpawnGroup(Masked Vector=<0 0 -16>, Passthrough Vector=<0 0 0>)
    Position Offset Coordinate Space=::World
    Sprite Size=<130 72>
    Sprite Size Mode=::NewEnumerator4
    UsePositionOffset=true
    Write Material Random=true
   ) {module=InitializeParticle001}
   &DynamicMaterialParameters(
    Index 0 Param 1=0.2
    Module.Param0WriteEnabled=true
    Module.Param1WriteEnabled=true
    Module.Param2WriteEnabled=true
    Module.Param3WriteEnabled=true
   )
   &CameraOffset(Camera Offset Amount=6) {module=CameraOffset001}
   /SetVariables(Particles.float=180) {isSetParameters=true}
  ]
  particleUpdateScript [
   &ParticleState()
   &SolveForcesAndVelocity(Clamp Velocity=false, Limit Acceleration=false)
   &SocketLocation(
    SamplePosOffset=true
    Sampled Position Offset=<0 1.5 0>
    Socket Index=&SelectIntFromArray(Direct Array Index=$Particles.PartitionParticles.Partition, Int Selection Array=DataInterface() {propertyValues {IntData [0,1]}})
    Socket Reader=$System.Socket Reader
    Socket Sampling Mode=::NewEnumerator3
    VelocitySpeedLimit=false
   )
   &ScaleColor(
    Scale Alpha=&MaskFloat(
     Mask Source=1
     Mask Parameter=$Particles.PartitionParticles.Partition
     Mask=0
     Masked Float=&Multiply_Float(
      A=1
      B=&MakeCustomFloatFromBool(
       Bool=&SetBoolByFloatComparison(A=$User.FXActive, B=0, Comparison Type=::NewEnumerator6)
       True Float=&Multiply_Float(A=$User.FXActive, B=-1)
       False Float=$User.FXActive
      )
     )
     Passthrough Float=&Multiply_Float(
      A=1
      B=&Multiply_Float(
       A=1
       B=&MakeCustomFloatFromBool(
        Bool=&SetBoolByFloatComparison(A=$User.FXColor, B=0, Comparison Type=::NewEnumerator6)
        True Float=&Multiply_Float(A=$User.FXColor, B=-1)
        False Float=$User.FXColor
       )
      )
     )
    )
    ScaleA=true
    ScaleRGB=true
   )
   /SetVariables(
    Particles.SparkColor:LinearColor=&MakeLinearColorFromVectorAndFloat(
     "Vector (RGB)"=&Multiply_VectorByFloat(
      Vector=&MakeVectorFromLinearColorRGB(Color=<rgba 15 13.4273 4.4675 1>)
      Float=&MakeCustomFloatFromBool(
       Bool=&SetBoolByFloatComparison(
        A=&FN_PickFloatDependOnPlatform(
         TargetPlatform=DataInterface() {propertyValues {Platforms {
            qualityLevelMask=-30
            deviceProfileStates [
             {profileName=Windows, qualityLevelMask=0, setQualityLevelMask=3}
             {profileName=Mac, qualityLevelMask=0, setQualityLevelMask=3}
             {profileName=Linux, qualityLevelMask=0, setQualityLevelMask=3}
             {profileName=IOS, qualityLevelMask=0, setQualityLevelMask=4294967295}
             {profileName=Mobile, qualityLevelMask=0, setQualityLevelMask=4294967295}
            ]
           }}}
         TargetPlatformActive=1
         TargetPlatformNoActive=0
        )
        B=0
       )
       True Float=0.4
       False Float=1
      )
     )
     "Float (Alpha)"=&MaskFloat(
      Mask Source=1
      Mask Parameter=$Particles.PartitionParticles.Partition
      Mask=0
      Masked Float=&Multiply_Float(A=1, B=$User.FXActive)
      Passthrough Float=&Multiply_Float(A=1, B=&Multiply_Float(A=1, B=$User.FXColor))
     )
    )
    Particles.SparkSize:Vector2f=&MakeVector2D(
     X=29.5912
     Y=&ClampFloat(Float=&Sine(Normalized Angle=$Emitter.Age, Period=0.0616, Scale=4.1323, Bias=4.3498), Min=12.5264, Max=24.152)
    )
    Particles.Vector 2D:Vector2f=&Vector2DFromCurve(
     Vector2Curve=DataInterface() {propertyValues {
       XCurve {defaultValue=1, keys [{tangentMode=RCTM_User, time=0, value=0.5}, {tangentMode=RCTM_User, time=1, value=1}]}
       YCurve {defaultValue=1, keys [{tangentMode=RCTM_User, time=0, value=0.5}, {tangentMode=RCTM_User, time=1, value=1}]}
      }}
     ScaleCurve=&RandomRangeVector2D(Minimum=<18 6.7222>, Maximum=<18 6.4088>)
    )
   ) {isSetParameters=true}
   /SetVariables(Particles.Linear Color=&MakeLinearColorFromVectorAndFloat(
     "Vector (RGB)"=&MakeVectorFromLinearColorRGB(Color=<rgba 5 1.7194 0 1>)
     "Float (Alpha)"=&MaskFloat(
      Mask Source=1
      Mask Parameter=$Particles.PartitionParticles.Partition
      Mask=0
      Masked Float=&Multiply_Float(
       A=1
       B=&MakeCustomFloatFromBool(
        Bool=&SetBoolByFloatComparison(A=$User.FXActive, B=0, Comparison Type=::NewEnumerator6)
        True Float=&Multiply_Float(A=$User.FXActive, B=-1)
        False Float=$User.FXActive
       )
      )
      Passthrough Float=&Multiply_Float(
       A=1
       B=&Multiply_Float(
        A=1
        B=&MakeCustomFloatFromBool(
         Bool=&SetBoolByFloatComparison(A=$User.FXColor, B=0, Comparison Type=::NewEnumerator6)
         True Float=&Multiply_Float(A=$User.FXColor, B=-1)
         False Float=$User.FXColor
        )
       )
      )
     )
    )) {isSetParameters=true}
   /SetVariables(Particles.Supporting Flare Color:LinearColor=&MakeLinearColorFromVectorAndFloat(
     "Vector (RGB)"=&Multiply_VectorByFloat(
      Vector=&MakeVectorFromLinearColorRGB(Color=<rgba 5 1.7194 0 1>)
      Float=&MakeCustomFloatFromBool(
       Bool=&SetBoolByFloatComparison(
        A=&FN_PickFloatDependOnPlatform(
         TargetPlatform=DataInterface() {propertyValues {Platforms {
            qualityLevelMask=-29
            deviceProfileStates [
             {profileName=Windows, qualityLevelMask=0, setQualityLevelMask=2}
             {profileName=Mac, qualityLevelMask=0, setQualityLevelMask=3}
             {profileName=Linux, qualityLevelMask=0, setQualityLevelMask=3}
            ]
           }}}
         TargetPlatformActive=1
         TargetPlatformNoActive=0
        )
        B=0
       )
       True Float=0.25
       False Float=1
      )
     )
     "Float (Alpha)"=&MaskFloat(
      Mask Source=1
      Mask Parameter=$Particles.PartitionParticles.Partition
      Mask=0
      Masked Float=&Multiply_Float(A=1, B=$User.FXActive)
      Passthrough Float=&Multiply_Float(A=1, B=&Multiply_Float(A=1, B=$User.FXColor))
     )
    )) {isSetParameters=true}
   /SetVariables(Particles.Secondary Size:Vector2f=&Vector2DFromCurve(
     Vector2Curve=DataInterface() {propertyValues {
       XCurve {defaultValue=1, keys [{tangentMode=RCTM_User, time=0, value=0.5}, {tangentMode=RCTM_User, time=1, value=1}]}
       YCurve {defaultValue=1, keys [{tangentMode=RCTM_User, time=0, value=0.5}, {tangentMode=RCTM_User, time=1, value=1}]}
      }}
     ScaleCurve=<28 23>
    )) {isSetParameters=true}
   &SubUVAnimation(
    Random Start Frame=true
    "Sprite Renderer "=DataInterface() {propertyValues {SpriteRenderer=&self:Flash_Flare_0.NiagaraSpriteRendererProperties_1}}
    SubUV Lookup Index Scale=5.7547
    UseEndFrame=false
    UseStartFrame=false
   )
   /SetVariables(Particles.Position=&ConvertPositionToVector(
     Input Position=&AddVectorToPosition(
      Position=$Output.SocketLocation.SampledPosition
      Vector=&MaskVectorBySpawnGroup(Masked Vector=<0 1 -2>, Passthrough Vector=<0 1 0>)
     )
     Conversion Operation:ENiagara_LWCConvertPosToVec=::NewEnumerator0
    )) {isSetParameters=true}
   &FN_Light_Attributes(
    Color=<rgba 2 1.0278 0 1>
    Create Color Attribute=true
    Create Radius Attribute=true
    Radius=&Multiply_Float(
     A=70
     B=&Multiply_Float(
      A=1
      B=&MaskFloat(
       Mask Source=1
       Mask Parameter=$Particles.PartitionParticles.Partition
       Mask=0
       Masked Float=&Multiply_Float(
        A=1
        B=&MakeCustomFloatFromBool(
         Bool=&SetBoolByFloatComparison(A=$User.FXActive, B=0, Comparison Type=::NewEnumerator6)
         True Float=&Multiply_Float(A=$User.FXActive, B=-1.5)
         False Float=$User.FXActive
        )
       )
       Passthrough Float=&Multiply_Float(
        A=1
        B=&Multiply_Float(
         A=1
         B=&MakeCustomFloatFromBool(
          Bool=&SetBoolByFloatComparison(A=$User.FXColor, B=0, Comparison Type=::NewEnumerator6)
          True Float=&Multiply_Float(A=$User.FXColor, B=-1.5)
          False Float=$User.FXColor
         )
        )
       )
      )
     )
    )
    Scale Color=&Multiply_Float(A=1.5, B=$System.FlickerSpeed)
   )
   /SetVariables(Particles.PositionLights:NiagaraPosition=&AddVectorToPosition(Position=$Output.SocketLocation.SampledPosition, Vector=<0 -20 10>)) {isSetParameters=true}
  ]
  rendererClasses [&NiagaraSpriteRendererProperties, &NiagaraLightRendererProperties]
  renderers [
   {rendererClass=&NiagaraSpriteRendererProperties}
   {rendererClass=&NiagaraLightRendererProperties, rendererIndex=1}
   {rendererClass=&NiagaraSpriteRendererProperties, rendererIndex=2}
   {rendererClass=&NiagaraSpriteRendererProperties, rendererIndex=3}
   {rendererClass=&NiagaraSpriteRendererProperties, rendererIndex=4}
  ]
 }
}
summary {
 systemName=NS_Emote_JoyGrid
 emitters [
  {emitterName=Flash_Flare, rendererClasses [&NiagaraSpriteRendererProperties, &NiagaraLightRendererProperties]}
  {emitterName=Classic_Sparks, rendererClasses [&NiagaraSpriteRendererProperties]}
  {emitterName=Decal, rendererClasses [&NiagaraDecalRendererProperties]}
 ]
 userVariables [
  User.FXActive {type=NiagaraFloat, defaultValue=0, description=""}
  User.FXColor {type=NiagaraFloat, defaultValue=0, description=""}
  User.OwnerLocalVelocity {type=Vector3f, description="", defaultValue {struct=Vector3f, value=<0 0 0>}}
 ]
}
